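# Niagara effect for the CraftGlue Hearts player emote, adding glowing pink slice streaks and twinkling sparkles to a skeletal mesh.
# Slice bursts five meshes two seconds in, using partition index to pick each one's bone or socket, scale, tilt and offset.
# Sparkles spawn in a squashed sphere around another socket, tapering off after two seconds.
# Waveform-modulated alpha makes them flicker, and gravity keyed to emitter age later pulls them down.
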NiagaraSystem /BRCosmetics/Animation/Game/MainPlayer/Emotes/CraftGlue/FX/NS_Emote_CraftGlue_Hearts
dependencies {
 usedDataInterfaces [
  &NiagaraDataInterfaceArrayFloat3
  &NiagaraDataInterfaceArrayPosition
  &NiagaraDataInterfaceCurve
  &NiagaraDataInterfaceSkeletalMesh
 ]
 usedDynamicInputs [&RandomRangeVector]
 usedModules [
  &self:SystemSpawnScript.SystemScriptSource.SystemScriptGraph.NiagaraNodeAssignment_0.SetVariables_#1
  &SystemState
 ]
}
emitters {
 Slice {
  emitterUpdateScript [
   &EmitterState()
   &SpawnBurst_Instantaneous(Spawn Count=5, Spawn Time=2, Use Spawn Probability=false, UseLoopCountLimit=false)
  ]
  particleSpawnScript [
   &PartitionParticles(Number Of Partitions=5)
   &InitializeParticle(
    Color=<rgba 9.8708 7.4351 10 1>
    Mesh Scale=&SelectVectorFromArray(
     Direct Array Index=$Particles.PartitionParticles.Partition
     Vector Selection Array=DataInterface() {propertyValues {FloatData [<10 40 1>, <8 30 1>, <5 20 1>, <8 30 1>, <6 25 1>]}}
    )
    Mesh Scale Mode=::NewEnumerator4
    UsePositionOffset=true
    Write Material Random=true
   )
   &CameraOffset(Camera Offset Amount=10) {enabled=false}
   &InitialMeshOrientation(
    Mesh Axis To Orient=&SelectVectorFromArray(
     Direct Array Index=$Particles.PartitionParticles.Partition
     Vector Selection Array=DataInterface() {propertyValues {FloatData [<0.5 0 0.25>, <0.5 0 0.33>, <0.5 0 0.5>, <0.5 0 0.6>, <0.5 0 0.8>]}}
    )
    Mesh Orientation Mode=::NewEnumerator2
    Orientation Vector=<0 1 0>
    Rotation Coordinate Space=::NewEnumerator2
    Use Rotation Vector=false
   )
   &SkeletalMeshLocation(
    Bone / Socket Index=$Particles.PartitionParticles.Partition
    Bone Sampling Mode=::NewEnumerator4
    MaskSpawnGroup=false
    Sampled Position Offset=/Niagara/DynamicInputs/Arrays/SelectPositionFromArray(
     Array Sampling Mode:ENiagaraArraySamplingMode=::NewEnumerator1
     Direct Array Index:NiagaraInt32=$Particles.PartitionParticles.Partition
     Interpolation:NiagaraFloat=?
     Direct Array Mode:ENiagara_WrapClamp=::NewEnumerator0
     Evaluation Type:ENiagaraRandomnessEvaluation=::NewEnumerator0
     Normalized Interpolation:NiagaraBool=true
     Recalculate Random Each Loop:NiagaraBool=false
     Position Array:NiagaraDataInterfaceArrayPosition=DataInterface() {propertyValues {GpuSyncMode=SyncCpuToGpu, MaxElements=0, PositionData [<0 0 1>, <0 0 2>, <0 0 0>, <0 0 -0.5>, <0 0 0>]}}
     Randomness Mode:ENiagaraRandomnessMode=::NewEnumerator0
     Fixed Random Seed:NiagaraBool=false
     Random Seed:NiagaraInt32=?
    )
    Sampled Position Offset Coordinate Space:ENiagaraSkelMeshTransforms=::NewEnumerator2
    Skeletal Mesh=$System.Skeletal Mesh
   )
  ]
  particleUpdateScript [
   &ParticleState()
   &ScaleColor(
    Scale Alpha=&FloatFromCurve(FloatCurve=DataInterface() {propertyValues {Curve {keys [
         {arriveTangent=9.0725, interpMode=RCIM_Cubic, leaveTangent=9.0725, tangentMode=RCTM_User, time=0, value=0}
         {arriveTangent=-2.5249, interpMode=RCIM_Cubic, leaveTangent=-2.5249, tangentMode=RCTM_User, time=0.1, value=1}
         {arriveTangent=-0.0405, interpMode=RCIM_Cubic, leaveTangent=-0.0405, tangentMode=RCTM_User, time=1, value=0}
        ]}}})
    ScaleA=true
    ScaleRGB=true
   )
  ]
  rendererClasses [&NiagaraMeshRendererProperties]
  renderers [{rendererClass=&NiagaraMeshRendererProperties, rendererIndex=0}]
 }
 Sparkles {
  emitterUpdateScript [
   &EmitterState()
   &SpawnRate(
    SpawnRate=&FloatFromCurve(FloatCurve=DataInterface() {propertyValues {Curve {keys [<at 2 1>,<at 3 0>]}}}, CurveIndex=$Emitter.Age, Scale Curve=7)
    Use Spawn Probability=false
   )
  ]
  particleSpawnScript [
   &InitializeParticle(
    Color Maximum=<rgba 1 0.75 0.95 1>
    Color Minimum=<rgba 0.98 0.35 1 1>
    Color Mode=::NewEnumerator2
    Lifetime Max=2.25
    Lifetime Min=1.5
    Lifetime Mode=::NewEnumerator1
    Sprite Rotation Mode=::NewEnumerator2
    Sprite Size Mode=::NewEnumerator1
    Sprite UV Mode=::NewEnumerator3
    Uniform Sprite Size Max=30
    Uniform Sprite Size Min=10
    UsePositionOffset=false
    Write Material Random=true
   )
   &SkeletalMeshLocation(
    Bone / Socket Index=5
    Bone Sampling Mode=::NewEnumerator4
    MaskSpawnGroup=false
    Sampled Position Offset=<-3 25 135>
    Sampled Position Offset Coordinate Space:ENiagaraSkelMeshTransforms=::NewEnumerator2
    Skeletal Mesh=$System.Skeletal Mesh
   )
   &ShapeLocation(MaskSpawnGroup=false, Non Uniform Scale=<1 0.75 0.75>, Override Seed=false, Sphere Radius=42, UseOwnerScale=true)
  ]
  particleUpdateScript [
   &ParticleState()
   /SetVariables(Particles.SoftGlow_Color:LinearColor=&RandomRangeLinearColor(
     Minimum=&MakeLinearColor(
      R:NiagaraFloat=0.98
      G:NiagaraFloat=0.35
      B:NiagaraFloat=1
      A:NiagaraFloat=&FloatFromCurve(
       FloatCurve=DataInterface() {propertyValues {Curve {keys [
           <at 0 0 RCIM_Cubic>
           {arriveTangent=1.0526, interpMode=RCIM_Cubic, leaveTangent=1.0526, time=0.15, value=1}
           {tangentMode=RCTM_User, time=0.95, value=1}
           {tangentMode=RCTM_User, time=1, value=0}
          ]}}}
       Scale Curve=&Waveform(
        "Waveform [1]"=::NewEnumerator0
        "[1] Phase Scale"=&RandomRangeFloat(Minimum=1, Maximum=2)
        "[1] Frequency"=1
        "[1] Amplitude Min/Max"=<0 0.5>
       )
      )
     )
     Maximum=&MakeLinearColor(
      R:NiagaraFloat=1
      G:NiagaraFloat=0.75
      B:NiagaraFloat=0.95
      A:NiagaraFloat=&FloatFromCurve(
       FloatCurve=DataInterface() {propertyValues {Curve {keys [
           <at 0 0 RCIM_Cubic>
           {arriveTangent=1.0526, interpMode=RCIM_Cubic, leaveTangent=1.0526, time=0.15, value=1}
           {tangentMode=RCTM_User, time=0.95, value=1}
           {tangentMode=RCTM_User, time=1, value=0}
          ]}}}
       Scale Curve=&Waveform(
        "Waveform [1]"=::NewEnumerator0
        "[1] Phase Scale"=&RandomRangeFloat(Minimum=1, Maximum=2)
        "[1] Frequency"=1
        "[1] Amplitude Min/Max"=<0 0.5>
       )
      )
     )
    )) {isSetParameters=true}
   &ScaleSpriteSize(Scale Factor=&Vector2DFromFloat(Value=&FloatFromCurve(FloatCurve=DataInterface() {propertyValues {Curve {keys [
          <at 0 0 RCIM_Cubic>
          <at 0.112 1.0173>
          {arriveTangent=-1.142, interpMode=RCIM_Cubic, leaveTangent=-1.142, time=0.7001, value=0.872}
          <at 1.0011 0.002>
         ]}}})))
   &ScaleColor(
    Scale Alpha=&FloatFromCurve(
     FloatCurve=DataInterface() {propertyValues {Curve {keys [
         <at 0 0 RCIM_Cubic>
         {arriveTangent=1.0526, interpMode=RCIM_Cubic, leaveTangent=1.0526, time=0.15, value=1}
         {tangentMode=RCTM_User, time=0.95, value=1}
         {tangentMode=RCTM_User, time=1, value=0}
        ]}}}
     Scale Curve=&Waveform(
      "Waveform [1]"=::NewEnumerator0
      "[1] Phase Scale"=&RandomRangeFloat(Minimum=0.5, Maximum=1)
      "[1] Frequency"=1
      "[1] Amplitude Min/Max"=<0 1>
     )
    )
    ScaleA=true
    ScaleRGB=true
   )
   &GravityForce(Gravity=&Lerp_Vector(
     A=<0 0 0>
     B=<0 0 -980>
     Alpha=&FloatFromCurve(
      FloatCurve=DataInterface() {propertyValues {Curve {keys [<at 3.5 0>,<at 4 1>]}}}
      CurveIndex=$Emitter.Age
      Scale Curve=&RandomRangeFloat(Minimum=0.1, Maximum=1)
     )
    ))
   &SolveForcesAndVelocity(Clamp Velocity=false, Limit Acceleration=false)
  ]
  rendererClasses [&NiagaraSpriteRendererProperties]
  renderers [
   {rendererClass=&NiagaraSpriteRendererProperties}
   {rendererClass=&NiagaraSpriteRendererProperties, rendererIndex=1}
  ]
 }
}
summary {
 systemName=NS_Emote_CraftGlue_Hearts
 emitters [
  {emitterName=Slice, rendererClasses [&NiagaraMeshRendererProperties]}
  {emitterName=Sparkles, rendererClasses [&NiagaraSpriteRendererProperties]}
 ]
 userVariables []
}
