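# Niagara ribbon trail for the knife in the CraftGlue emote, driven by the animated skeletal mesh passed in User.SkeletalMesh.
# Two FN_AnimTrail emitters layer a translucent pale teal trail (two ribbon renderers) with a black one, both multiplied by a time-of-day tint.
# A curve scaled by User.FXTrailScale, which defaults to zero, sets the short particle lifetime.
# Alpha fades out early in each particle's life, following a different curve per layer.
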NiagaraSystem /BRCosmetics/Animation/Game/MainPlayer/Emotes/CraftGlue/FX/NS_Emote_CraftGlue_KnifeTrail
dependencies {
 usedDataInterfaces [&NiagaraDataInterfaceCurve]
 usedDynamicInputs []
 usedModules [
  &self:SystemSpawnScript.SystemScriptSource.SystemScriptGraph.NiagaraNodeAssignment_0.SetVariables_#1
  &SystemState
 ]
}
emitters {
 FN_AnimTrail {
  emitterSpawnScript [&FN_RibbonIDOffsetUtility()]
  emitterUpdateScript [
   &EmitterState()
   &FN_SpawnAnimTrail()
  ]
  particleSpawnScript [
   &FN_InitializeAnimTrailBoneIndecies()
   &FN_InitializeAnimTrail(
    Color=<rgba 0.3459 0.4427 0.4427 0.4>
    Lifetime=&Lerp_Float(
     A=0
     B=0.18
     Alpha=&FloatFromCurve(
      FloatCurve=DataInterface() {propertyValues {Curve {keys [
          {arriveTangent=-1.836, interpMode=RCIM_Cubic, leaveTangent=-1.836, tangentMode=RCTM_User, time=0, value=1}
          {interpMode=RCIM_Cubic, tangentMode=RCTM_User, time=1, value=0}
         ]}}}
      Scale Curve=$User.FXTrailScale
     )
    )
   )
   &Color(Color=&MultiplyLinearColor(
     Linear Color=$Particles.Initial.Color
     Scale Factor=&MakeVector4_FromLinearColor(LinearColor=&FN_Select_Color_from_Time_Of_Day(Morning=<rgba 0.6719 0.6019 0.6097 1>, Day=<rgba 1 1 1 1.1>, Evening=<rgba 1 0.6413 0.7956 1>, Night=<rgba 0.4668 0.9064 2 1>))
    )) {module=Color001}
  ]
  particleUpdateScript [
   &SolveForcesAndVelocity(Clamp Velocity=false, Limit Acceleration=false)
   &FN_UpdateAnimTrail(Locked Particle RibbonWidth=1, SkeletalMesh=$User.SkeletalMesh) {module=FN_UpdateAnimTrail001}
   &ScaleColor(
    Scale Alpha=&FloatFromCurve(FloatCurve=DataInterface() {propertyValues {Curve {keys [{interpMode=RCIM_Cubic, leaveTangent=-4.0138, tangentMode=RCTM_Break, time=0, value=1}, <at 0.4 0>]}}})
    ScaleA=true
    ScaleRGB=true
   )
  ]
  rendererClasses [&NiagaraRibbonRendererProperties]
  renderers [
   {rendererClass=&NiagaraRibbonRendererProperties}
   {rendererClass=&NiagaraRibbonRendererProperties, rendererIndex=1}
  ]
 }
 FN_AnimTrail001 {
  emitterSpawnScript [&FN_RibbonIDOffsetUtility()]
  emitterUpdateScript [
   &EmitterState()
   &FN_SpawnAnimTrail()
  ]
  particleSpawnScript [
   &FN_InitializeAnimTrailBoneIndecies()
   &FN_InitializeAnimTrail(
    Color=<rgba 0 0 0 1>
    Lifetime=&Lerp_Float(
     A=0
     B=0.18
     Alpha=&FloatFromCurve(
      FloatCurve=DataInterface() {propertyValues {Curve {keys [
          {arriveTangent=-1.836, interpMode=RCIM_Cubic, leaveTangent=-1.836, tangentMode=RCTM_User, time=0, value=1}
          {interpMode=RCIM_Cubic, tangentMode=RCTM_User, time=1, value=0}
         ]}}}
      Scale Curve=$User.FXTrailScale
     )
    )
   )
   &Color(Color=&MultiplyLinearColor(
     Linear Color=$Particles.Initial.Color
     Scale Factor=&MakeVector4_FromLinearColor(LinearColor=&FN_Select_Color_from_Time_Of_Day(Morning=<rgba 0.6719 0.6019 0.6097 1>, Day=<rgba 1 1 1 1.1>, Evening=<rgba 1 0.6413 0.7956 1>, Night=<rgba 0.4668 0.9064 2 1>))
    )) {module=Color001}
  ]
  particleUpdateScript [
   &SolveForcesAndVelocity(Clamp Velocity=false, Limit Acceleration=false)
   &FN_UpdateAnimTrail(Locked Particle RibbonWidth=1, SkeletalMesh=$User.SkeletalMesh) {module=FN_UpdateAnimTrail001}
   &ScaleColor(
    Scale Alpha=&FloatFromCurve(FloatCurve=DataInterface() {propertyValues {Curve {keys [
         {interpMode=RCIM_Cubic, leaveTangent=-0.3449, tangentMode=RCTM_Break, time=0, value=1}
         {arriveTangent=-5.2668, tangentMode=RCTM_User, time=0.4, value=0}
        ]}}})
    ScaleA=true
    ScaleRGB=true
   )
  ]
  rendererClasses [&NiagaraRibbonRendererProperties]
  renderers [{rendererClass=&NiagaraRibbonRendererProperties}]
 }
}
summary {
 systemName=NS_Emote_CraftGlue_KnifeTrail
 emitters [
  {emitterName=FN_AnimTrail, rendererClasses [&NiagaraRibbonRendererProperties]}
  {emitterName=FN_AnimTrail001, rendererClasses [&NiagaraRibbonRendererProperties]}
 ]
 userVariables [
  User.bIsFrontEnd {type=NiagaraBool}
  User.bIsFrontEndPreview {type=NiagaraBool}
  User.FXTrailScale {type=NiagaraFloat, defaultValue=0, description=""}
  User.RibbonIDOffset {type=NiagaraInt32}
  User.SkeletalMesh {
   type=NiagaraDataInterfaceSkeletalMesh
   defaultValue {
    struct=NiagaraExt_VariableValue_DataInterface
    value {dataInterface=&self:NiagaraDataInterfaceSkeletalMesh_0}
   }
  }
 ]
}
